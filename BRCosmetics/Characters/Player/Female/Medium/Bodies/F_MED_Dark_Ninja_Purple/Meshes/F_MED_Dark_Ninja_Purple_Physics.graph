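# Physics asset for the female medium Dark Ninja Purple player body, modeling two dangling lace chains plus the lower body and spine.
# Capsules cover the pelvis, four spine bones, thighs, calves and feet, joined by constraints with free swing and twist; arms and head have no bodies.
# Large and small laces are four-box chains, each rooted on a tiny offset capsule freely constrained to the pelvis, with non-free angular constraints along each chain.
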PhysicsAsset /BRCosmetics/Characters/Player/Female/Medium/Bodies/F_MED_Dark_Ninja_Purple/Meshes/F_MED_Dark_Ninja_Purple_Physics
bodies {
 calf_l [{shapeType=Capsule, length=31.6737, radius=8.0673, center=<-23.2198 -2.0157 0>, rotation=<rot 90 0 0>}]
 calf_r [{shapeType=Capsule, length=29.8352, radius=6.9105, center=<23.3177 2.5513 -0.0004>, rotation=<rot 90 0 0>}]
 dyn_LargeLace_01 [{shapeType=Box, extentX=3.5322, extentY=12.8626, extentZ=1.0748, center=<1.6001 -0.0001 -0.0578>}]
 dyn_LargeLace_01_Offset [{shapeType=Capsule, length=1, radius=1}]
 dyn_LargeLace_02 [{shapeType=Box, extentX=3.5322, extentY=12.8626, extentZ=1.7188, center=<1.6001 0.0001 0.0704>, rotation=<rot 0 0 15.4161>}]
 dyn_LargeLace_03 [{shapeType=Box, extentX=3.5322, extentY=12.8626, extentZ=1.7188, center=<1.6001 0 0>, rotation=<rot 0.0001 0 15.416>}]
 dyn_LargeLace_04 [{shapeType=Box, extentX=3.5322, extentY=12.8626, extentZ=1.7188, center=<1.6001 0 0>, rotation=<rot 0 0 15.416>}]
 dyn_SmallLace_01 [{shapeType=Box, extentX=4.3182, extentY=14.707, extentZ=2.484, center=<2.0812 0 0.0002>}]
 dyn_SmallLace_02 [{shapeType=Box, extentX=4.3182, extentY=14.707, extentZ=2.8417, center=<2.0812 0 0.0002>}]
 dyn_SmallLace_03 [{shapeType=Box, extentX=4.3182, extentY=14.707, extentZ=3.0938, center=<2.0812 0 0.0002>}]
 dyn_SmallLace_04 [{shapeType=Box, extentX=4.3182, extentY=14.707, extentZ=3.1842, center=<2.0812 0 0.0002>}]
 dyn_SmallLace_Outside_Offset [{shapeType=Capsule, length=1, radius=1}]
 foot_l [{shapeType=Capsule, length=14.5631, radius=6.4462, center=<-3.5855 7.24 0.0001>, rotation=<rot 90 0 78.8422>}]
 foot_r [{shapeType=Capsule, length=13.1131, radius=7.8962, center=<3.4126 -6.243 0.0003>, rotation=<rot 90 0 78.8406>}]
 pelvis [
  {shapeType=Capsule, length=11.2627, radius=9.372, center=<-2.8542 -3.3615 0>}
  {shapeType=Capsule, length=13.456, radius=4.521, center=<-2.5136 5.2776 -0.4864>, rotation=<rot 0 0 26.8773>}
 ]
 spine_01 [{shapeType=Capsule, length=12.739, radius=7.6772}]
 spine_02 [{shapeType=Capsule, length=12.739, radius=7.6772}]
 spine_03 [{shapeType=Capsule, length=12.739, radius=7.6772}]
 spine_04 [{shapeType=Capsule, length=12.739, radius=7.6772}]
 thigh_l [{shapeType=Capsule, length=30.9656, radius=7.1549, center=<-21.9638 -0.0002 0>, rotation=<rot 90 0 0>}]
 thigh_r [{shapeType=Capsule, length=33.5195, radius=6.7751, center=<23.3217 0.0007 0>, rotation=<rot 90 0 0>}]
}
constraints [
 {
  bone1Name=spine_01
  bone2Name=pelvis
  swing1LimitDegrees=45
  swing1Motion=Free
  swing2LimitDegrees=45
  swing2Motion=Free
  twistLimitDegrees=45
  twistMotion=Free
 }
 {
  bone1Name=spine_02
  bone2Name=spine_01
  swing1LimitDegrees=45
  swing1Motion=Free
  swing2LimitDegrees=45
  swing2Motion=Free
  twistLimitDegrees=45
  twistMotion=Free
 }
 {
  bone1Name=spine_03
  bone2Name=spine_02
  swing1LimitDegrees=45
  swing1Motion=Free
  swing2LimitDegrees=45
  swing2Motion=Free
  twistLimitDegrees=45
  twistMotion=Free
 }
 {
  bone1Name=spine_04
  bone2Name=spine_03
  swing1LimitDegrees=45
  swing1Motion=Free
  swing2LimitDegrees=45
  swing2Motion=Free
  twistLimitDegrees=45
  twistMotion=Free
 }
 {
  bone1Name=thigh_l
  bone2Name=pelvis
  swing1LimitDegrees=45
  swing1Motion=Free
  swing2LimitDegrees=45
  swing2Motion=Free
  twistLimitDegrees=45
  twistMotion=Free
 }
 {
  bone1Name=calf_l
  bone2Name=thigh_l
  swing1LimitDegrees=45
  swing1Motion=Free
  swing2LimitDegrees=45
  swing2Motion=Free
  twistLimitDegrees=45
  twistMotion=Free
 }
 {
  bone1Name=foot_l
  bone2Name=calf_l
  swing1LimitDegrees=45
  swing1Motion=Free
  swing2LimitDegrees=45
  swing2Motion=Free
  twistLimitDegrees=45
  twistMotion=Free
 }
 {
  bone1Name=thigh_r
  bone2Name=pelvis
  swing1LimitDegrees=45
  swing1Motion=Free
  swing2LimitDegrees=45
  swing2Motion=Free
  twistLimitDegrees=45
  twistMotion=Free
 }
 {
  bone1Name=calf_r
  bone2Name=thigh_r
  swing1LimitDegrees=45
  swing1Motion=Free
  swing2LimitDegrees=45
  swing2Motion=Free
  twistLimitDegrees=45
  twistMotion=Free
 }
 {
  bone1Name=foot_r
  bone2Name=calf_r
  swing1LimitDegrees=45
  swing1Motion=Free
  swing2LimitDegrees=45
  swing2Motion=Free
  twistLimitDegrees=45
  twistMotion=Free
 }
 {bone1Name=dyn_LargeLace_01, bone2Name=dyn_LargeLace_01_Offset, swing1LimitDegrees=45, swing2LimitDegrees=45, twistLimitDegrees=20}
 {bone1Name=dyn_LargeLace_02, bone2Name=dyn_LargeLace_01, swing1LimitDegrees=45, swing2LimitDegrees=45, twistLimitDegrees=20}
 {bone1Name=dyn_LargeLace_03, bone2Name=dyn_LargeLace_02, swing1LimitDegrees=45, swing2LimitDegrees=45, twistLimitDegrees=20}
 {bone1Name=dyn_LargeLace_04, bone2Name=dyn_LargeLace_03, swing1LimitDegrees=45, swing2LimitDegrees=45, twistLimitDegrees=20}
 {
  bone1Name=dyn_LargeLace_01_Offset
  bone2Name=pelvis
  swing1LimitDegrees=45
  swing1Motion=Free
  swing2LimitDegrees=45
  swing2Motion=Free
  twistLimitDegrees=45
  twistMotion=Free
 }
 {
  bone1Name=dyn_SmallLace_Outside_Offset
  bone2Name=pelvis
  swing1LimitDegrees=45
  swing1Motion=Free
  swing2LimitDegrees=45
  swing2Motion=Free
  twistLimitDegrees=45
  twistMotion=Free
 }
 {bone1Name=dyn_SmallLace_01, bone2Name=dyn_SmallLace_Outside_Offset, swing1LimitDegrees=45, swing2LimitDegrees=45, twistLimitDegrees=20}
 {bone1Name=dyn_SmallLace_02, bone2Name=dyn_SmallLace_01, swing1LimitDegrees=45, swing2LimitDegrees=45, twistLimitDegrees=20}
 {bone1Name=dyn_SmallLace_03, bone2Name=dyn_SmallLace_02, swing1LimitDegrees=45, swing2LimitDegrees=45, twistLimitDegrees=20}
 {bone1Name=dyn_SmallLace_04, bone2Name=dyn_SmallLace_03, swing1LimitDegrees=45, swing2LimitDegrees=45, twistLimitDegrees=20}
]
