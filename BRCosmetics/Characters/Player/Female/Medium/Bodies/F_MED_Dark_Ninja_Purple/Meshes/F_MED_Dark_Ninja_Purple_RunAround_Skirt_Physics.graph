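# Physics asset for the skirt of the Dark Ninja Purple female medium character, by name its RunAround variant.
# Capsules on spine, pelvis, legs and feet form the body, with an extra wider capsule on the left thigh.
# Five chains of box bodies (front, front-side, side, back-side, back) share identical swing and twist limits per link.
# Small offset capsules root the chains with free-rotation constraints: front pair to left thigh, others to pelvis.
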PhysicsAsset /BRCosmetics/Characters/Player/Female/Medium/Bodies/F_MED_Dark_Ninja_Purple/Meshes/F_MED_Dark_Ninja_Purple_RunAround_Skirt_Physics
bodies {
 calf_l [{shapeType=Capsule, length=32.8755, radius=6.8655, center=<-23.2198 -2.8865 0>, rotation=<rot 90 0 -4.5746>}]
 calf_r [{shapeType=Capsule, length=29.8352, radius=6.9105, center=<23.3177 2.5513 -0.0004>, rotation=<rot 90 0 0>}]
 dyn_SkirtBackSide_01 [{shapeType=Box, extentX=14.644, extentY=10, extentZ=2, center=<7.335 0 0>}]
 dyn_SkirtBackSide_02 [{shapeType=Box, extentX=13.3289, extentY=15, extentZ=2, center=<6.8669 0 0>}]
 dyn_SkirtBackSide_03 [{shapeType=Box, extentX=10, extentY=20, extentZ=2, center=<6.8806 0 -0.0001>}]
 dyn_SkirtBackSide_04 [{shapeType=Box, extentX=13.6076, extentY=20, extentZ=2, center=<6.8728 0 0>}]
 dyn_SkirtBackSide_Offset [{shapeType=Capsule, length=1, radius=1}]
 dyn_SkirtBack_01 [{shapeType=Box, extentX=13.8929, extentY=10, extentZ=2, center=<6.9167 0 -0.0003>}]
 dyn_SkirtBack_02 [{shapeType=Box, extentX=12.6363, extentY=15, extentZ=2, center=<6.294 0 -0.0003>}]
 dyn_SkirtBack_03 [{shapeType=Box, extentX=10.2971, extentY=15, extentZ=2, center=<5.2798 0 -0.0003>}]
 dyn_SkirtBack_04 [{shapeType=Box, extentX=13.8929, extentY=15, extentZ=2, center=<6.9167 0 -0.0003>}]
 dyn_SkirtBack_Offset [{shapeType=Capsule, length=1, radius=1}]
 dyn_SkirtFrontSide_01 [{shapeType=Box, extentX=9, extentY=5, extentZ=2, center=<9.6575 0 -0.4664>, rotation=<rot -22.9759 0.0001 0>}]
 dyn_SkirtFrontSide_02 [{shapeType=Box, extentX=11, extentY=15, extentZ=2, center=<7.1808 -0.0001 0>}]
 dyn_SkirtFrontSide_03 [{shapeType=Box, extentX=11, extentY=15, extentZ=2, center=<6.0159 -0.0001 0.0002>}]
 dyn_SkirtFrontSide_04 [{shapeType=Box, extentX=9.8681, extentY=15, extentZ=2, center=<5.0564 -0.0001 0.0001>}]
 dyn_SkirtFrontSide_05 [{shapeType=Box, extentX=10, extentY=20, extentZ=2, center=<6.004 -0.0001 -0.4363>, rotation=<rot -3.1182 0 0>}]
 dyn_SkirtFrontSide_Offset [{shapeType=Capsule, length=1, radius=1}]
 dyn_SkirtFront_01 [{shapeType=Box, extentX=15, extentY=10, extentZ=2, center=<8.1919 0 0.7665>, rotation=<rot -4.4514 0.0001 0>}]
 dyn_SkirtFront_02 [{shapeType=Box, extentX=10.5, extentY=9.5, extentZ=3.5, center=<7.4139 0 0.0649>, rotation=<rot 0.0001 0.0001 10.2078>}]
 dyn_SkirtFront_03 [{shapeType=Box, extentX=11, extentY=8, extentZ=5, center=<6.0856 0 0.3348>}]
 dyn_SkirtFront_04 [{shapeType=Box, extentX=15.7949, extentY=17.2611, extentZ=4, center=<7.9463 0.0001 -0.6969>, rotation=<rot -3.2898 0 0>}]
 dyn_SkirtFront_05 [{shapeType=Box, extentX=10.9463, extentY=17.2611, extentZ=4, center=<5.165 0.0001 -0.5371>, rotation=<rot -3.2898 0 0>}]
 dyn_SkirtFront_Offset [{shapeType=Capsule, length=1, radius=1}]
 dyn_SkirtSide_01 [{shapeType=Box, extentX=11, extentY=5, extentZ=2, center=<7.9128 0 -0.0002>}]
 dyn_SkirtSide_02 [{shapeType=Box, extentX=11, extentY=20, extentZ=2, center=<6.7232 0 -0.0002>}]
 dyn_SkirtSide_03 [{shapeType=Box, extentX=11, extentY=20, extentZ=2, center=<6.0328 0 0.0002>}]
 dyn_SkirtSide_04 [{shapeType=Box, extentX=11, extentY=20, extentZ=4, center=<4.956 0 0.0001>}]
 dyn_SkirtSide_05 [{shapeType=Box, extentX=8, extentY=20, extentZ=8, center=<4.7362 0.0001 -0.9885>, rotation=<rot -12.5328 0 0>}]
 dyn_SkirtSide_Offset [{shapeType=Capsule, length=1, radius=1}]
 foot_l [{shapeType=Capsule, length=15.5922, radius=5.4171, center=<-3.5855 7.24 0.0001>, rotation=<rot 90 0 78.8422>}]
 foot_r [{shapeType=Capsule, length=14.5183, radius=6.4909, center=<3.4126 -6.243 0.0003>, rotation=<rot 90 0 78.8406>}]
 pelvis [{shapeType=Capsule, length=16.1165, radius=4.5182}]
 spine_01 [{shapeType=Capsule, length=12.739, radius=7.6772}]
 spine_02 [{shapeType=Capsule, length=12.739, radius=7.6772}]
 spine_03 [{shapeType=Capsule, length=12.739, radius=7.6772}]
 spine_04 [{shapeType=Capsule, length=12.739, radius=7.6772}]
 thigh_l [
  {shapeType=Capsule, length=31.1072, radius=7.0134, center=<-21.9638 -0.0004 0.4394>, rotation=<rot 90 0 0>}
  {
   shapeType=Capsule
   length=26.7298
   radius=11.3907
   shapeName=thigh_l_capsule
   center=<-21.9624 -2.6866 2.963>
   rotation=<rot 90 -19.7672 -21.2693>
  }
 ]
 thigh_r [{shapeType=Capsule, length=33.5195, radius=6.7751, center=<23.3217 0.0007 0>, rotation=<rot 90 0 0>}]
}
constraints [
 {
  bone1Name=spine_01
  bone2Name=pelvis
  swing1LimitDegrees=45
  swing1Motion=Free
  swing2LimitDegrees=45
  swing2Motion=Free
  twistLimitDegrees=45
  twistMotion=Free
 }
 {
  bone1Name=spine_02
  bone2Name=spine_01
  swing1LimitDegrees=45
  swing1Motion=Free
  swing2LimitDegrees=45
  swing2Motion=Free
  twistLimitDegrees=45
  twistMotion=Free
 }
 {
  bone1Name=spine_03
  bone2Name=spine_02
  swing1LimitDegrees=45
  swing1Motion=Free
  swing2LimitDegrees=45
  swing2Motion=Free
  twistLimitDegrees=45
  twistMotion=Free
 }
 {
  bone1Name=spine_04
  bone2Name=spine_03
  swing1LimitDegrees=45
  swing1Motion=Free
  swing2LimitDegrees=45
  swing2Motion=Free
  twistLimitDegrees=45
  twistMotion=Free
 }
 {
  bone1Name=thigh_l
  bone2Name=pelvis
  swing1LimitDegrees=45
  swing1Motion=Free
  swing2LimitDegrees=45
  swing2Motion=Free
  twistLimitDegrees=45
  twistMotion=Free
 }
 {
  bone1Name=calf_l
  bone2Name=thigh_l
  swing1LimitDegrees=45
  swing1Motion=Free
  swing2LimitDegrees=45
  swing2Motion=Free
  twistLimitDegrees=45
  twistMotion=Free
 }
 {
  bone1Name=foot_l
  bone2Name=calf_l
  swing1LimitDegrees=45
  swing1Motion=Free
  swing2LimitDegrees=45
  swing2Motion=Free
  twistLimitDegrees=45
  twistMotion=Free
 }
 {
  bone1Name=thigh_r
  bone2Name=pelvis
  swing1LimitDegrees=45
  swing1Motion=Free
  swing2LimitDegrees=45
  swing2Motion=Free
  twistLimitDegrees=45
  twistMotion=Free
 }
 {
  bone1Name=calf_r
  bone2Name=thigh_r
  swing1LimitDegrees=45
  swing1Motion=Free
  swing2LimitDegrees=45
  swing2Motion=Free
  twistLimitDegrees=45
  twistMotion=Free
 }
 {
  bone1Name=foot_r
  bone2Name=calf_r
  swing1LimitDegrees=45
  swing1Motion=Free
  swing2LimitDegrees=45
  swing2Motion=Free
  twistLimitDegrees=45
  twistMotion=Free
 }
 {bone1Name=dyn_SkirtFront_01, bone2Name=dyn_SkirtFront_Offset, swing1LimitDegrees=45, swing2LimitDegrees=25, twistLimitDegrees=20}
 {bone1Name=dyn_SkirtFront_02, bone2Name=dyn_SkirtFront_01, swing1LimitDegrees=45, swing2LimitDegrees=25, twistLimitDegrees=20}
 {bone1Name=dyn_SkirtFront_03, bone2Name=dyn_SkirtFront_02, swing1LimitDegrees=45, swing2LimitDegrees=25, twistLimitDegrees=20}
 {bone1Name=dyn_SkirtFront_04, bone2Name=dyn_SkirtFront_03, swing1LimitDegrees=45, swing2LimitDegrees=25, twistLimitDegrees=20}
 {bone1Name=dyn_SkirtFront_05, bone2Name=dyn_SkirtFront_04, swing1LimitDegrees=45, swing2LimitDegrees=25, twistLimitDegrees=20}
 {bone1Name=dyn_SkirtFrontSide_01, bone2Name=dyn_SkirtFrontSide_Offset, swing1LimitDegrees=45, swing2LimitDegrees=25, twistLimitDegrees=20}
 {bone1Name=dyn_SkirtFrontSide_02, bone2Name=dyn_SkirtFrontSide_01, swing1LimitDegrees=45, swing2LimitDegrees=25, twistLimitDegrees=20}
 {bone1Name=dyn_SkirtFrontSide_03, bone2Name=dyn_SkirtFrontSide_02, swing1LimitDegrees=45, swing2LimitDegrees=25, twistLimitDegrees=20}
 {bone1Name=dyn_SkirtFrontSide_04, bone2Name=dyn_SkirtFrontSide_03, swing1LimitDegrees=45, swing2LimitDegrees=25, twistLimitDegrees=20}
 {bone1Name=dyn_SkirtFrontSide_05, bone2Name=dyn_SkirtFrontSide_04, swing1LimitDegrees=45, swing2LimitDegrees=25, twistLimitDegrees=20}
 {bone1Name=dyn_SkirtSide_01, bone2Name=dyn_SkirtSide_Offset, swing1LimitDegrees=45, swing2LimitDegrees=25, twistLimitDegrees=20}
 {bone1Name=dyn_SkirtSide_02, bone2Name=dyn_SkirtSide_01, swing1LimitDegrees=45, swing2LimitDegrees=25, twistLimitDegrees=20}
 {bone1Name=dyn_SkirtSide_03, bone2Name=dyn_SkirtSide_02, swing1LimitDegrees=45, swing2LimitDegrees=25, twistLimitDegrees=20}
 {bone1Name=dyn_SkirtSide_04, bone2Name=dyn_SkirtSide_03, swing1LimitDegrees=45, swing2LimitDegrees=25, twistLimitDegrees=20}
 {bone1Name=dyn_SkirtSide_05, bone2Name=dyn_SkirtSide_04, swing1LimitDegrees=45, swing2LimitDegrees=25, twistLimitDegrees=20}
 {bone1Name=dyn_SkirtBackSide_01, bone2Name=dyn_SkirtBackSide_Offset, swing1LimitDegrees=45, swing2LimitDegrees=25, twistLimitDegrees=20}
 {bone1Name=dyn_SkirtBackSide_02, bone2Name=dyn_SkirtBackSide_01, swing1LimitDegrees=45, swing2LimitDegrees=25, twistLimitDegrees=20}
 {bone1Name=dyn_SkirtBackSide_03, bone2Name=dyn_SkirtBackSide_02, swing1LimitDegrees=45, swing2LimitDegrees=25, twistLimitDegrees=20}
 {bone1Name=dyn_SkirtBackSide_04, bone2Name=dyn_SkirtBackSide_03, swing1LimitDegrees=45, swing2LimitDegrees=25, twistLimitDegrees=20}
 {bone1Name=dyn_SkirtBack_01, bone2Name=dyn_SkirtBack_Offset, swing1LimitDegrees=45, swing2LimitDegrees=25, twistLimitDegrees=20}
 {bone1Name=dyn_SkirtBack_02, bone2Name=dyn_SkirtBack_01, swing1LimitDegrees=45, swing2LimitDegrees=25, twistLimitDegrees=20}
 {bone1Name=dyn_SkirtBack_03, bone2Name=dyn_SkirtBack_02, swing1LimitDegrees=45, swing2LimitDegrees=25, twistLimitDegrees=20}
 {bone1Name=dyn_SkirtBack_04, bone2Name=dyn_SkirtBack_03, swing1LimitDegrees=45, swing2LimitDegrees=25, twistLimitDegrees=20}
 {
  bone1Name=dyn_SkirtFront_Offset
  bone2Name=thigh_l
  swing1LimitDegrees=45
  swing1Motion=Free
  swing2LimitDegrees=45
  swing2Motion=Free
  twistLimitDegrees=45
  twistMotion=Free
 }
 {
  bone1Name=dyn_SkirtFrontSide_Offset
  bone2Name=thigh_l
  swing1LimitDegrees=45
  swing1Motion=Free
  swing2LimitDegrees=45
  swing2Motion=Free
  twistLimitDegrees=45
  twistMotion=Free
 }
 {
  bone1Name=dyn_SkirtSide_Offset
  bone2Name=pelvis
  swing1LimitDegrees=45
  swing1Motion=Free
  swing2LimitDegrees=45
  swing2Motion=Free
  twistLimitDegrees=45
  twistMotion=Free
 }
 {
  bone1Name=dyn_SkirtBackSide_Offset
  bone2Name=pelvis
  swing1LimitDegrees=45
  swing1Motion=Free
  swing2LimitDegrees=45
  swing2Motion=Free
  twistLimitDegrees=45
  twistMotion=Free
 }
 {
  bone1Name=dyn_SkirtBack_Offset
  bone2Name=pelvis
  swing1LimitDegrees=45
  swing1Motion=Free
  swing2LimitDegrees=45
  swing2Motion=Free
  twistLimitDegrees=45
  twistMotion=Free
 }
]
