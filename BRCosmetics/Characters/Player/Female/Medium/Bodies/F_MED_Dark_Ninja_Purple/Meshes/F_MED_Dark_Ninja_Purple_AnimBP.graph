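# Anim Blueprint for the Dark Ninja Purple female body, adding secondary motion to skirt and lace bones over the parent mesh's copied pose.
# Ground poses use bone constraints, then crouch-selected RigidBody setups, with AnimDynamics laces when crouched.
# A separate skydive/swim/parachute RigidBody branch takes external force from skydive state and vertical velocity.
# The cached result gets extra skirt constraints when jumping or falling, plus bone offsets when downed with rigid bodies disabled.
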AnimBlueprint /BRCosmetics/Characters/Player/Female/Medium/Bodies/F_MED_Dark_Ninja_Purple/Meshes/F_MED_Dark_Ninja_Purple_AnimBP
target_skeleton=&/F_MED_Dark_Ninja_Purple_Skeleton
graphs [
 AnimGraph {nodes [
   "Constraint - dyn_sideCloth_r (P)"(Alpha=0.5, ConstraintWeights_0=1) {
    alphaBoolBlend {blendOption=Linear}
    boneToModify {boneName=dyn_sideCloth_r}
    constraintSetup [{targetBone {boneName=thigh_r}}]
    constraintWeights [1]
   }
   "Transform (Modify) Bone - Bone: dyn_SkirtSide_Offset"(Alpha=1, Translation="100,0,4", Rotation="44,0,0", Scale="1,1,1") {
    rotationMode=BMM_Additive
    rotationSpace=BCS_BoneSpace
    translationMode=BMM_Additive
    translationSpace=BCS_BoneSpace
    alphaBoolBlend {blendOption=Linear}
    boneToModify {boneName=dyn_SkirtSide_Offset}
    rotation {pitch=44}
    translation {x=100, z=4}
   }
   RigidBody() {
    bClampLinearTranslationLimitToRefPose=true
    bOverrideWorldGravity=true
    overridePhysicsAsset=&/F_MED_Dark_Ninja_Purple_Physics
    simulationSpace=BaseBoneSpace
    alphaBoolBlend {blendOption=Linear}
    baseBoneRef {boneName=root}
    componentAppliedLinearAccClamp=<5000 5000 5000>
    componentLinearAccScale=<0.6 0.6 0.6>
    overrideWorldGravity {z=-980}
   }
   n1 = "Constraint - dyn_SkirtSide_Offset (PPT)"!(
    ComponentPose=Local To Component(LocalPose=Use cached pose 'PreJump'())
    Alpha=1
    ConstraintWeights_0=0.7
    ConstraintWeights_1:float=0.3
    ConstraintWeights_2:float=0
   ) -> (Pose) {
    alphaBoolBlend {blendOption=Linear}
    boneToModify {boneName=dyn_SkirtSide_Offset}
    constraintSetup [
     {targetBone {boneName=dyn_SkirtFrontSide_Offset}}
     {targetBone {boneName=dyn_SkirtBackSide_Offset}}
     {transformType=Translation, targetBone {boneName=None}}
    ]
    constraintWeights [1,1,1]
   }
   n2 = "Constraint - dyn_skirtSIde_01 (PPT)"!(ComponentPose=@n1, Alpha=1, ConstraintWeights_0=0.7, ConstraintWeights_1:float=0.3, ConstraintWeights_2:float=0) -> (Pose) {
    alphaBoolBlend {blendOption=Linear}
    boneToModify {boneName=dyn_skirtSIde_01}
    constraintSetup [
     {targetBone {boneName=dyn_skirtFrontSIde_01}}
     {targetBone {boneName=dyn_skirtBackSIde_01}}
     {transformType=Translation, targetBone {boneName=None}}
    ]
    constraintWeights [1,1,1]
   }
   n3 = "Constraint - dyn_skirtSIde_02 (PPT)"!(ComponentPose=@n2, Alpha=1, ConstraintWeights_0=0.7, ConstraintWeights_1:float=0.3, ConstraintWeights_2:float=0) -> (Pose) {
    alphaBoolBlend {blendOption=Linear}
    boneToModify {boneName=dyn_skirtSIde_02}
    constraintSetup [
     {targetBone {boneName=dyn_skirtFrontSIde_02}}
     {targetBone {boneName=dyn_skirtBackSIde_02}}
     {transformType=Translation, targetBone {boneName=None}}
    ]
    constraintWeights [1,1,1]
   }
   n4 = "Constraint - dyn_skirtSIde_03 (PPT)"!(ComponentPose=@n3, Alpha=1, ConstraintWeights_0=0.7, ConstraintWeights_1:float=0.3, ConstraintWeights_2:float=0) -> (Pose) {
    alphaBoolBlend {blendOption=Linear}
    boneToModify {boneName=dyn_skirtSIde_03}
    constraintSetup [
     {targetBone {boneName=dyn_skirtFrontSIde_03}}
     {targetBone {boneName=dyn_skirtBackSIde_03}}
     {transformType=Translation, targetBone {boneName=None}}
    ]
    constraintWeights [1,1,1]
   }
   n5 = "Constraint - dyn_skirtSIde_04 (PPT)"!(ComponentPose=@n4, Alpha=1, ConstraintWeights_0=0.7, ConstraintWeights_1:float=0.3, ConstraintWeights_2:float=0) -> (Pose) {
    alphaBoolBlend {blendOption=Linear}
    boneToModify {boneName=dyn_skirtSIde_04}
    constraintSetup [
     {targetBone {boneName=dyn_skirtFrontSIde_04}}
     {targetBone {boneName=dyn_skirtBackSIde_04}}
     {transformType=Translation, targetBone {boneName=None}}
    ]
    constraintWeights [1,1,1]
   }
   n6 = "Constraint - dyn_SkirtBackSide_Offset (PPT)"!(ComponentPose=@n5, Alpha=1, ConstraintWeights_0=0.5, ConstraintWeights_1:float=0.5, ConstraintWeights_2:float=0) -> (Pose) {
    alphaBoolBlend {blendOption=Linear}
    boneToModify {boneName=dyn_SkirtBackSide_Offset}
    constraintSetup [
     {targetBone {boneName=dyn_SkirtSide_Offset}}
     {targetBone {boneName=dyn_SkirtBack_Offset}}
     {transformType=Translation, targetBone {boneName=None}}
    ]
    constraintWeights [1,1,1]
   }
   n7 = "Constraint - dyn_skirtBackSIde_01 (PPT)"!(ComponentPose=@n6, Alpha=1, ConstraintWeights_0=0.5, ConstraintWeights_1:float=0.5, ConstraintWeights_2:float=0) -> (Pose) {
    alphaBoolBlend {blendOption=Linear}
    boneToModify {boneName=dyn_skirtBackSIde_01}
    constraintSetup [{targetBone {boneName=dyn_skirtSIde_01}}, {targetBone {boneName=dyn_skirtBack_01}}, {transformType=Translation, targetBone {boneName=None}}]
    constraintWeights [1,1,1]
   }
   n8 = "Constraint - dyn_skirtBackSIde_02 (PPT)"!(ComponentPose=@n7, Alpha=1, ConstraintWeights_0=0.5, ConstraintWeights_1:float=0.5, ConstraintWeights_2:float=0) -> (Pose) {
    alphaBoolBlend {blendOption=Linear}
    boneToModify {boneName=dyn_skirtBackSIde_02}
    constraintSetup [{targetBone {boneName=dyn_skirtSIde_02}}, {targetBone {boneName=dyn_skirtBack_02}}, {transformType=Translation, targetBone {boneName=None}}]
    constraintWeights [1,1,1]
   }
   n9 = "Constraint - dyn_skirtBackSIde_03 (PPT)"!(ComponentPose=@n8, Alpha=1, ConstraintWeights_0=0.5, ConstraintWeights_1:float=0.5, ConstraintWeights_2:float=0) -> (Pose) {
    alphaBoolBlend {blendOption=Linear}
    boneToModify {boneName=dyn_skirtBackSIde_03}
    constraintSetup [{targetBone {boneName=dyn_skirtSIde_03}}, {targetBone {boneName=dyn_skirtBack_03}}, {transformType=Translation, targetBone {boneName=None}}]
    constraintWeights [1,1,1]
   }
   n10 = "Constraint - dyn_skirtBackSIde_04 (PPT)"!(ComponentPose=@n9, Alpha=1, ConstraintWeights_0=0.5, ConstraintWeights_1:float=0.5, ConstraintWeights_2:float=0) -> (Pose) {
    alphaBoolBlend {blendOption=Linear}
    boneToModify {boneName=dyn_skirtBackSIde_04}
    constraintSetup [{targetBone {boneName=dyn_skirtSIde_04}}, {targetBone {boneName=dyn_skirtBack_04}}, {transformType=Translation, targetBone {boneName=None}}]
    constraintWeights [1,1,1]
   }
   n11 = "Constraint - dyn_SkirtBack_Offset (P)"(ComponentPose=@n10, Alpha=0.3, ConstraintWeights_0=1) {
    alphaBoolBlend {blendOption=Linear}
    boneToModify {boneName=dyn_SkirtBack_Offset}
    constraintSetup [{targetBone {boneName=dyn_SkirtBackSide_Offset}}]
    constraintWeights [1]
   }
   n12 = "Constraint - dyn_SkirtBack_01 (P)"(ComponentPose=@n11, Alpha=0.3, ConstraintWeights_0=1) {
    alphaBoolBlend {blendOption=Linear}
    boneToModify {boneName=dyn_SkirtBack_01}
    constraintSetup [{targetBone {boneName=dyn_SkirtBackSide_01}}]
    constraintWeights [1]
   }
   n13 = "Constraint - dyn_SkirtBack_02 (P)"(ComponentPose=@n12, Alpha=0.3, ConstraintWeights_0=1) {
    alphaBoolBlend {blendOption=Linear}
    boneToModify {boneName=dyn_SkirtBack_02}
    constraintSetup [{targetBone {boneName=dyn_SkirtBackSide_02}}]
    constraintWeights [1]
   }
   n14 = "Constraint - dyn_SkirtBack_03 (P)"(ComponentPose=@n13, Alpha=0.2, ConstraintWeights_0=1) {
    alphaBoolBlend {blendOption=Linear}
    boneToModify {boneName=dyn_SkirtBack_03}
    constraintSetup [{targetBone {boneName=dyn_SkirtBackSide_03}}]
    constraintWeights [1]
   }
   n15 = "Constraint - dyn_SkirtBack_04 (P)"(ComponentPose=@n14, Alpha=0.1, ConstraintWeights_0=1) {
    alphaBoolBlend {blendOption=Linear}
    boneToModify {boneName=dyn_SkirtBack_04}
    constraintSetup [{targetBone {boneName=dyn_SkirtBackSide_04}}]
    constraintWeights [1]
   }
   n16 = Component To Local(ComponentPose=@n15)
   n17 = Use cached pose 'PreJump'()
   n18 = Get falling_notskydiving() -> (falling_notskydiving:bool)
   n19 = Get rban_disabled_dbno() -> (rban_disabled_dbno:bool)
   n20 = "Transform (Modify) Bone - Bone: dyn_SkirtFront_Offset"!(
    ComponentPose=Local To Component(LocalPose=Blend Poses by bool(BlendPose_0=@n16, BlendPose_1=@n17, bActiveValue=@n18) {blendType=Linear})
    bAlphaBoolEnabled:bool=@n19
    Translation="0,0,0"
    Rotation="10,0,0"
    Scale="1,1,1"
   ) -> (Pose) {
    alphaInputType=Bool
    rotationMode=BMM_Additive
    rotationSpace=BCS_BoneSpace
    scaleMode=BMM_Additive
    translationMode=BMM_Additive
    alphaBoolBlend {blendInTime=0.25, blendOption=Linear, blendOutTime=0.25}
    boneToModify {boneName=dyn_SkirtFront_Offset}
    rotation {pitch=10}
   }
   n21 = "Transform (Modify) Bone - Bone: dyn_SkirtFrontSide_Offset"!(ComponentPose=@n20, bAlphaBoolEnabled:bool=@n19, Translation="0,0,0", Rotation="10,0,0", Scale="1,1,1") -> (Pose) {
    alphaInputType=Bool
    rotationMode=BMM_Additive
    rotationSpace=BCS_BoneSpace
    scaleMode=BMM_Additive
    translationMode=BMM_Additive
    alphaBoolBlend {blendInTime=0.25, blendOption=Linear, blendOutTime=0.25}
    boneToModify {boneName=dyn_SkirtFrontSide_Offset}
    rotation {pitch=10}
   }
   n22 = "Transform (Modify) Bone - Bone: dyn_SkirtBackSide_Offset"!(ComponentPose=@n21, bAlphaBoolEnabled:bool=@n19, Translation="0,0,0", Rotation="10,0,0", Scale="1,1,1") -> (Pose) {
    alphaInputType=Bool
    rotationMode=BMM_Additive
    rotationSpace=BCS_BoneSpace
    scaleMode=BMM_Additive
    translationMode=BMM_Additive
    alphaBoolBlend {blendInTime=0.25, blendOption=Linear, blendOutTime=0.25}
    boneToModify {boneName=dyn_SkirtBackSide_Offset}
    rotation {pitch=10}
   }
   n23 = "Transform (Modify) Bone - Bone: dyn_SkirtBack_Offset"!(ComponentPose=@n22, bAlphaBoolEnabled:bool=@n19, Translation="0,0,0", Rotation="10,0,0", Scale="1,1,1") -> (Pose) {
    alphaInputType=Bool
    rotationMode=BMM_Additive
    rotationSpace=BCS_BoneSpace
    scaleMode=BMM_Additive
    translationMode=BMM_Additive
    alphaBoolBlend {blendInTime=0.25, blendOption=Linear, blendOutTime=0.25}
    boneToModify {boneName=dyn_SkirtBack_Offset}
    rotation {pitch=10}
   }
   n24 = "Transform (Modify) Bone - Bone: dyn_SkirtSide_Offset"!(ComponentPose=@n23, bAlphaBoolEnabled:bool=@n19, Translation="0,0,0", Rotation="10,0.000,0", Scale="1,1,1") -> (Pose) {
    alphaInputType=Bool
    rotationMode=BMM_Additive
    rotationSpace=BCS_BoneSpace
    scaleMode=BMM_Additive
    translationMode=BMM_Additive
    alphaBoolBlend {blendInTime=0.25, blendOption=Linear, blendOutTime=0.25}
    boneToModify {boneName=dyn_SkirtSide_Offset}
    rotation {pitch=10}
   }
   n25 = "Transform (Modify) Bone - Bone: dyn_LargeLace_01_Offset"!(ComponentPose=@n24, bAlphaBoolEnabled:bool=@n19, Translation="0,0,0", Rotation="40,0.000,-10", Scale="1,1,1") -> (Pose) {
    alphaInputType=Bool
    rotationMode=BMM_Additive
    rotationSpace=BCS_BoneSpace
    scaleMode=BMM_Additive
    translationMode=BMM_Additive
    alphaBoolBlend {blendInTime=0.25, blendOption=Linear, blendOutTime=0.25}
    boneToModify {boneName=dyn_LargeLace_01_Offset}
    rotation {pitch=40, roll=-10}
   }
   n26 = "Transform (Modify) Bone - Bone: dyn_SmallLace_Outside_Offset"!(ComponentPose=@n25, bAlphaBoolEnabled:bool=@n19, Translation="0,0,0", Rotation="25,0.000,-10", Scale="1,1,1") -> (Pose) {
    alphaInputType=Bool
    rotationMode=BMM_Additive
    rotationSpace=BCS_BoneSpace
    scaleMode=BMM_Additive
    translationMode=BMM_Additive
    alphaBoolBlend {blendInTime=0.25, blendOption=Linear, blendOutTime=0.25}
    boneToModify {boneName=dyn_SmallLace_Outside_Offset}
    rotation {pitch=25, roll=-10}
   }
   Output Pose(Result=Component To Local(ComponentPose=@n26)) {layerGroup=None}
   n27 = "Constraint - dyn_LargeLace_01_Offset (R)"(ComponentPose=Local To Component(LocalPose=Use cached pose 'CrossConstraints'()), Alpha=0.25, ConstraintWeights_0=1) {
    alphaBoolBlend {blendOption=Linear}
    boneToModify {boneName=dyn_LargeLace_01_Offset}
    constraintSetup [{transformType=Rotation, targetBone {boneName=thigh_twist_01_l}}]
    constraintWeights [1]
   }
   n28 = Get bIsCrouching()
   n29 = "Transform (Modify) Bone - Bone: dyn_LargeLace_01_Offset"(ComponentPose=@n27, bAlphaBoolEnabled=@n28, Translation="0,-4,6.5", Rotation="0,-50,0", Scale="1,1,1") {
    alphaInputType=Bool
    rotationMode=BMM_Additive
    scaleMode=BMM_Additive
    translationMode=BMM_Additive
    alphaBoolBlend {blendOption=Linear}
    boneToModify {boneName=dyn_LargeLace_01_Offset}
    rotation {yaw=-50}
    translation {y=-4, z=6.5}
   }
   n30 = "Constraint - dyn_LargeLace_05 (P)"(ComponentPose=@n29, Alpha=1, ConstraintWeights_0=1) {
    alphaBoolBlend {blendOption=Linear}
    boneToModify {boneName=dyn_LargeLace_05}
    constraintSetup [{targetBone {boneName=pelvis}}]
    constraintWeights [1]
   }
   n31 = "To Float (Boolean)"(InBool=Get bRigidBodyEnabled())
   n32 = "Transform (Modify) Bone - Bone: dyn_SkirtSide_Offset"!(ComponentPose=@n30, Alpha:float=@n31, Translation="0,-5,0", Rotation="50,0,0", Scale="1,1,1") -> (Pose) {
    rotationMode=BMM_Additive
    rotationSpace=BCS_BoneSpace
    translationMode=BMM_Additive
    translationSpace=BCS_BoneSpace
    alphaBoolBlend {blendOption=Linear}
    boneToModify {boneName=dyn_SkirtSide_Offset}
    rotation {pitch=50}
    translation {y=-5}
   }
   n33 = "Transform (Modify) Bone - Bone: dyn_SkirtBackSide_Offset"!(ComponentPose=@n32, Alpha:float=@n31, Translation="0,-6,0", Rotation="50,0,0", Scale="1,1,1") -> (Pose) {
    rotationMode=BMM_Additive
    rotationSpace=BCS_BoneSpace
    translationMode=BMM_Additive
    translationSpace=BCS_BoneSpace
    alphaBoolBlend {blendOption=Linear}
    boneToModify {boneName=dyn_SkirtBackSide_Offset}
    rotation {pitch=50}
    translation {y=-6}
   }
   n34 = RigidBody(ComponentPose=@n33) {
    bClampLinearTranslationLimitToRefPose=true
    bOverrideWorldGravity=true
    overridePhysicsAsset=&/F_MED_Dark_Ninja_Purple_RunAround_Skirt_Physics
    simulationSpace=BaseBoneSpace
    alphaBoolBlend {blendOption=Linear}
    baseBoneRef {boneName=root}
    componentAppliedLinearAccClamp=<3750 3750 1500>
    componentLinearAccScale=<0.5 0.5 0.5>
    overrideWorldGravity {z=-1078}
   }
   n35 = RigidBody(ComponentPose=@n34) {
    bClampLinearTranslationLimitToRefPose=true
    bOverrideWorldGravity=true
    overridePhysicsAsset=&F_MED_Dark_Ninja_Purple_Physics
    simulationSpace=BaseBoneSpace
    alphaBoolBlend {blendOption=Linear}
    baseBoneRef {boneName=root}
    componentAppliedLinearAccClamp=<5000 5000 5000>
    componentLinearAccScale=<0.6 0.6 0.7>
    overrideWorldGravity {z=-980}
   }
   n36 = Component To Local(ComponentPose=@n35)
   n37 = "Constraint - dyn_LargeLace_01_Offset (R)"(ComponentPose=Local To Component(LocalPose=Use cached pose 'CrossConstraints'()), Alpha=0.25, ConstraintWeights_0=1) {
    alphaBoolBlend {blendOption=Linear}
    boneToModify {boneName=dyn_LargeLace_01_Offset}
    constraintSetup [{transformType=Rotation, targetBone {boneName=thigh_twist_01_l}}]
    constraintWeights [1]
   }
   n38 = "Transform (Modify) Bone - Bone: dyn_LargeLace_01_Offset"(ComponentPose=@n37, bAlphaBoolEnabled=Get bIsCrouching(), Translation="0,-6,10", Rotation="-20,-50,20", Scale="1,1,1") {
    alphaInputType=Bool
    rotationMode=BMM_Additive
    scaleMode=BMM_Additive
    translationMode=BMM_Additive
    alphaBoolBlend {blendOption=Linear}
    boneToModify {boneName=dyn_LargeLace_01_Offset}
    rotation=<rot -20 -50 20>
    translation {y=-6, z=10}
   }
   n39 = "Constraint - dyn_LargeLace_05 (P)"(ComponentPose=@n38, Alpha=1, ConstraintWeights_0=1) {
    alphaBoolBlend {blendOption=Linear}
    boneToModify {boneName=dyn_LargeLace_05}
    constraintSetup [{targetBone {boneName=pelvis}}]
    constraintWeights [1]
   }
   n40 = "Transform (Modify) Bone - Bone: dyn_LargeLace_01_Offset"(ComponentPose=@n39, bAlphaBoolEnabled=Get bIsSprinting(), Translation="0,1,0", Rotation="0,0,0", Scale="1,1,1") {
    alphaInputType=Bool
    rotationMode=BMM_Additive
    scaleMode=BMM_Additive
    translationMode=BMM_Additive
    alphaBoolBlend {blendOption=Linear}
    boneToModify {boneName=dyn_LargeLace_01_Offset}
    translation {y=1}
   }
   n41 = RigidBody(ComponentPose=@n40) {
    bClampLinearTranslationLimitToRefPose=true
    bOverrideWorldGravity=true
    overridePhysicsAsset=&/F_MED_Dark_Ninja_Purple_Skirt_Physics
    simulationSpace=BaseBoneSpace
    alphaBoolBlend {blendOption=Linear}
    baseBoneRef {boneName=root}
    componentAppliedLinearAccClamp=<25 25 25>
    overrideWorldGravity {z=-1000}
   }
   n42 = Anim Dynamics - Bone: dyn_SmallLace_01(ComponentPose=@n41, Alpha=0.3, ExternalForce="0,0,0") {
    angularDampingOverride=0.9999
    angularSpringConstant=75
    bAngularSpring=true
    bGravityOverrideInSimSpace=true
    bOverrideAngularDamping=true
    simulationSpace=RootRelative
    alphaBoolBlend {blendOption=Linear}
    boundBone {boneName=dyn_SmallLace_01}
    componentLinearAccScale=<1 1 1>
    componentLinearVelScale=<1 1 1>
    physicsBodyDefinitions [{
      sphereCollisionRadius=0
      boundBone {boneName=dyn_SmallLace_01}
      boxExtents=<20 10 10>
      constraintSetup {angularLimitsMax=<25 0 25>, angularLimitsMin=<-25 -25 -25>, angularTarget=<0 0 0>}
      localJointOffset=<2 0 0>
     }]
   }
   n43 = Anim Dynamics - Bone: dyn_SmallLace_02(ComponentPose=@n42, Alpha=0.3, ExternalForce="0,0,0") {
    angularDampingOverride=0.9999
    angularSpringConstant=75
    bAngularSpring=true
    bGravityOverrideInSimSpace=true
    bOverrideAngularDamping=true
    simulationSpace=RootRelative
    alphaBoolBlend {blendOption=Linear}
    boundBone {boneName=dyn_SmallLace_02}
    componentLinearAccScale=<1 1 1>
    componentLinearVelScale=<1 1 1>
    physicsBodyDefinitions [{
      sphereCollisionRadius=0
      boundBone {boneName=dyn_SmallLace_02}
      boxExtents=<20 10 10>
      constraintSetup {angularLimitsMax=<25 0 25>, angularLimitsMin=<-25 -25 -25>, angularTarget=<0 0 0>}
      localJointOffset=<2 0 0>
     }]
   }
   n44 = Anim Dynamics - Bone: dyn_SmallLace_03(ComponentPose=@n43, Alpha=0.3, ExternalForce="0,0,0") {
    angularDampingOverride=0.9999
    angularSpringConstant=75
    bAngularSpring=true
    bGravityOverrideInSimSpace=true
    bOverrideAngularDamping=true
    simulationSpace=RootRelative
    alphaBoolBlend {blendOption=Linear}
    boundBone {boneName=dyn_SmallLace_03}
    componentLinearAccScale=<1 1 1>
    componentLinearVelScale=<1 1 1>
    physicsBodyDefinitions [{
      sphereCollisionRadius=0
      boundBone {boneName=dyn_SmallLace_03}
      boxExtents=<20 10 10>
      constraintSetup {angularLimitsMax=<25 0 25>, angularLimitsMin=<-25 -25 -25>, angularTarget=<0 0 0>}
      localJointOffset=<2 0 0>
     }]
   }
   n45 = Anim Dynamics - Bone: dyn_LargeLace_01(ComponentPose=@n44, Alpha=0.3, ExternalForce="0,0,0") {
    angularDampingOverride=0.9999
    angularSpringConstant=75
    bAngularSpring=true
    bGravityOverrideInSimSpace=true
    bOverrideAngularDamping=true
    simulationSpace=RootRelative
    alphaBoolBlend {blendOption=Linear}
    boundBone {boneName=dyn_LargeLace_01}
    componentLinearAccScale=<1 1 1>
    componentLinearVelScale=<1 1 1>
    physicsBodyDefinitions [{
      sphereCollisionRadius=0
      boundBone {boneName=dyn_LargeLace_01}
      boxExtents=<20 10 10>
      constraintSetup {angularLimitsMax=<25 0 25>, angularLimitsMin=<-25 -25 -25>, angularTarget=<0 0 0>}
      localJointOffset=<2 0 0>
     }]
   }
   n46 = Anim Dynamics - Bone: dyn_LargeLace_02(ComponentPose=@n45, Alpha=0.3, ExternalForce="0,0,0") {
    angularDampingOverride=0.9999
    angularSpringConstant=75
    bAngularSpring=true
    bGravityOverrideInSimSpace=true
    bOverrideAngularDamping=true
    simulationSpace=RootRelative
    alphaBoolBlend {blendOption=Linear}
    boundBone {boneName=dyn_LargeLace_02}
    componentLinearAccScale=<1 1 1>
    componentLinearVelScale=<1 1 1>
    physicsBodyDefinitions [{
      sphereCollisionRadius=0
      boundBone {boneName=dyn_LargeLace_02}
      boxExtents=<20 10 10>
      constraintSetup {angularLimitsMax=<25 0 25>, angularLimitsMin=<-25 -25 -25>, angularTarget=<0 0 0>}
      localJointOffset=<2 0 0>
     }]
   }
   n47 = Anim Dynamics - Bone: dyn_LargeLace_03(ComponentPose=@n46, Alpha=0.3, ExternalForce="0,0,0") {
    angularDampingOverride=0.9999
    angularSpringConstant=75
    bAngularSpring=true
    bGravityOverrideInSimSpace=true
    bOverrideAngularDamping=true
    simulationSpace=RootRelative
    alphaBoolBlend {blendOption=Linear}
    boundBone {boneName=dyn_LargeLace_03}
    componentLinearAccScale=<1 1 1>
    componentLinearVelScale=<1 1 1>
    physicsBodyDefinitions [{
      sphereCollisionRadius=0
      boundBone {boneName=dyn_LargeLace_03}
      boxExtents=<20 10 10>
      constraintSetup {angularLimitsMax=<25 0 25>, angularLimitsMin=<-25 -25 -25>, angularTarget=<0 0 0>}
      localJointOffset=<2 0 0>
     }]
   }
   n48 = Component To Local(ComponentPose=@n47)
   n49 = Get bIsCrouching()
   n50 = "Transform (Modify) Bone - Bone: dyn_SkirtFront_Offset"!(
    ComponentPose=Local To Component(LocalPose=Blend Poses by bool(BlendPose_0=@n48, BlendPose_1=@n36, BlendTime_0=0.2, BlendTime_1=0.25, bActiveValue=@n49) {blendType=Linear})
    Alpha:float=1
    Translation="0,2,1"
    Rotation="0,0,0"
    Scale="1,1,1"
   ) -> (Pose) {
    rotationMode=BMM_Additive
    scaleMode=BMM_Additive
    translationMode=BMM_Additive
    alphaBoolBlend {blendOption=Linear}
    boneToModify {boneName=dyn_SkirtFront_Offset}
    translation {y=2, z=1}
   }
   n51 = Get bIsCrouching()
   n52 = "Transform (Modify) Bone - Bone: dyn_SkirtFront_01"!(ComponentPose=@n50, bAlphaBoolEnabled:bool=@n51, Translation="0,0,1", Rotation="0,0,0", Scale="1,1,1") -> (Pose) {
    alphaInputType=Bool
    rotationMode=BMM_Additive
    rotationSpace=BCS_BoneSpace
    scaleMode=BMM_Additive
    scaleSpace=BCS_BoneSpace
    translationMode=BMM_Additive
    translationSpace=BCS_BoneSpace
    alphaBoolBlend {blendOption=Linear}
    boneToModify {boneName=dyn_SkirtFront_01}
    translation {z=1}
   }
   n53 = "Transform (Modify) Bone - Bone: dyn_SkirtFront_Offset"!(ComponentPose=@n52, bAlphaBoolEnabled:bool=@n51, Translation="0,0,5", Rotation="0,0,0", Scale="1,1,1") -> (Pose) {
    alphaInputType=Bool
    rotationMode=BMM_Additive
    scaleMode=BMM_Additive
    translationMode=BMM_Additive
    alphaBoolBlend {blendOption=Linear}
    boneToModify {boneName=dyn_SkirtFront_Offset}
    translation {z=5}
   }
   n54 = "Transform (Modify) Bone - Bone: dyn_SkirtFrontSide_Offset"(ComponentPose=@n53, bAlphaBoolEnabled=@n51, Translation="0,0,5", Rotation="0,0,0", Scale="1,1,1") {
    alphaInputType=Bool
    rotationMode=BMM_Additive
    scaleMode=BMM_Additive
    translationMode=BMM_Additive
    alphaBoolBlend {blendOption=Linear}
    boneToModify {boneName=dyn_SkirtFrontSide_Offset}
    translation {z=5}
   }
   n55 = "Transform (Modify) Bone - Bone: dyn_SkirtSide_Offset"!(ComponentPose=@n54, bAlphaBoolEnabled:bool=@n51, Translation="2,0,5", Rotation="0,0,0", Scale="1,1,1") -> (Pose) {
    alphaInputType=Bool
    rotationMode=BMM_Additive
    scaleMode=BMM_Additive
    translationMode=BMM_Additive
    alphaBoolBlend {blendOption=Linear}
    boneToModify {boneName=dyn_SkirtSide_Offset}
    translation {x=2, z=5}
   }
   n56 = "Transform (Modify) Bone - Bone: dyn_SkirtBackSide_Offset"!(ComponentPose=@n55, bAlphaBoolEnabled:bool=@n51, Translation="2,0,5", Rotation="0,0,0", Scale="1,1,1") -> (Pose) {
    alphaInputType=Bool
    rotationMode=BMM_Additive
    scaleMode=BMM_Additive
    translationMode=BMM_Additive
    alphaBoolBlend {blendOption=Linear}
    boneToModify {boneName=dyn_SkirtBackSide_Offset}
    translation {x=2, z=5}
   }
   n57 = "Transform (Modify) Bone - Bone: dyn_SkirtBack_Offset"(ComponentPose=@n56, bAlphaBoolEnabled=@n51, Translation="0,0,5", Rotation="0,0,0", Scale="1,1,1") {
    alphaInputType=Bool
    rotationMode=BMM_Additive
    scaleMode=BMM_Additive
    translationMode=BMM_Additive
    alphaBoolBlend {blendOption=Linear}
    boneToModify {boneName=dyn_SkirtBack_Offset}
    translation {z=5}
   }
   n58 = RigidBody(
    ComponentPose=Local To Component(LocalPose=Copy Pose From Mesh() {bUseAttachedParent=true})
    ExternalForce=Get gravity_override() -> (gravity_override:Vector)
   ) {
    bClampLinearTranslationLimitToRefPose=true
    bOverrideWorldGravity=true
    overridePhysicsAsset=&/F_MED_Dark_Ninja_Purple_Skydive_Skirt_Physics
    simulationSpace=BaseBoneSpace
    alphaBoolBlend {blendOption=Linear}
    baseBoneRef {boneName=root}
    componentAppliedLinearAccClamp=<500 500 500>
    overrideWorldGravity {z=-980}
   }
   n59 = RigidBody(ComponentPose=@n58) {
    bClampLinearTranslationLimitToRefPose=true
    bOverrideWorldGravity=true
    overridePhysicsAsset=&F_MED_Dark_Ninja_Purple_Physics
    simulationSpace=BaseBoneSpace
    alphaBoolBlend {blendOption=Linear}
    baseBoneRef {boneName=root}
    componentAppliedLinearAccClamp=<5000 5000 5000>
    componentLinearAccScale=<0.6 0.6 0.6>
    overrideWorldGravity {z=-700}
   }
   n60 = "Transform (Modify) Bone - Bone: dyn_SkirtBack_Offset"!(ComponentPose=@n59, Alpha:float=1, Translation="0,0,0", Rotation="0.000,0.000,20", Scale="1,1,1") -> (Pose) {
    rotationMode=BMM_Additive
    scaleMode=BMM_Additive
    translationMode=BMM_Additive
    alphaBoolBlend {blendOption=Linear}
    boneToModify {boneName=dyn_SkirtBack_Offset}
    rotation {roll=20}
   }
   n61 = "Transform (Modify) Bone - Bone: dyn_SkirtBackSide_Offset"!(ComponentPose=@n60, Alpha:float=1, Translation="0,0,0", Rotation="0,0,10", Scale="1,1,1") -> (Pose) {
    rotationMode=BMM_Additive
    scaleMode=BMM_Additive
    translationMode=BMM_Additive
    alphaBoolBlend {blendOption=Linear}
    boneToModify {boneName=dyn_SkirtBackSide_Offset}
    rotation {roll=10}
   }
   n62 = Get parachute_opened_float() -> (parachute_opened_float:double)
   n63 = "Constraint - dyn_SkirtFront_Offset (P)"(ComponentPose=@n61, Alpha=@n62, ConstraintWeights_0=1) {
    alphaBoolBlend {blendOption=Linear}
    boneToModify {boneName=dyn_SkirtFront_Offset}
    constraintSetup [{targetBone {boneName=thigh_l}}]
    constraintWeights [1]
   }
   n64 = "Constraint - dyn_SkirtFront_01 (P)"(ComponentPose=@n63, Alpha=@n62, ConstraintWeights_0=1) {
    alphaBoolBlend {blendOption=Linear}
    boneToModify {boneName=dyn_SkirtFront_01}
    constraintSetup [{targetBone {boneName=thigh_l}}]
    constraintWeights [1]
   }
   n65 = "Constraint - dyn_SkirtFront_02 (P)"(ComponentPose=@n64, Alpha=@n62, ConstraintWeights_0=1) {
    alphaBoolBlend {blendOption=Linear}
    boneToModify {boneName=dyn_SkirtFront_02}
    constraintSetup [{targetBone {boneName=thigh_l}}]
    constraintWeights [1]
   }
   n66 = "Constraint - dyn_SkirtFront_03 (P)"(ComponentPose=@n65, Alpha=@n62, ConstraintWeights_0=1) {
    alphaBoolBlend {blendOption=Linear}
    boneToModify {boneName=dyn_SkirtFront_03}
    constraintSetup [{targetBone {boneName=thigh_l}}]
    constraintWeights [1]
   }
   n67 = "Constraint - dyn_SkirtFront_04 (P)"(ComponentPose=@n66, Alpha=@n62, ConstraintWeights_0=1) {
    alphaBoolBlend {blendOption=Linear}
    boneToModify {boneName=dyn_SkirtFront_04}
    constraintSetup [{targetBone {boneName=thigh_l}}]
    constraintWeights [1]
   }
   n68 = "Constraint - dyn_SkirtFront_05 (P)"(ComponentPose=@n67, Alpha=@n62, ConstraintWeights_0=1) {
    alphaBoolBlend {blendOption=Linear}
    boneToModify {boneName=dyn_SkirtFront_05}
    constraintSetup [{targetBone {boneName=thigh_l}}]
    constraintWeights [1]
   }
   Save cached pose 'PreJump'(Pose=Blend Poses by bool(
     BlendPose_0=Component To Local(ComponentPose=@n68)
     BlendPose_1=Component To Local(ComponentPose=@n57)
     bActiveValue=Get skydive_swim() -> (skydive_swim:bool)
    ) {blendType=Linear}) {cacheName=PreJump}
   n69 = "Constraint - dyn_SkirtSide_Offset (PPT)"(
    ComponentPose=Local To Component(LocalPose=Copy Pose From Mesh() {bUseAttachedParent=true})
    Alpha=1
    ConstraintWeights_0=0.7
    ConstraintWeights_1=0.3
    ConstraintWeights_2=0
   ) {
    alphaBoolBlend {blendOption=Linear}
    boneToModify {boneName=dyn_SkirtSide_Offset}
    constraintSetup [
     {targetBone {boneName=dyn_SkirtFrontSide_Offset}}
     {targetBone {boneName=dyn_SkirtBackSide_Offset}}
     {transformType=Translation, targetBone {boneName=None}}
    ]
    constraintWeights [1,1,1]
   }
   n70 = "Constraint - dyn_skirtSIde_01 (PPT)"(ComponentPose=@n69, Alpha=1, ConstraintWeights_0=0.7, ConstraintWeights_1=0.3, ConstraintWeights_2=0) {
    alphaBoolBlend {blendOption=Linear}
    boneToModify {boneName=dyn_skirtSIde_01}
    constraintSetup [
     {targetBone {boneName=dyn_skirtFrontSIde_01}}
     {targetBone {boneName=dyn_skirtBackSIde_01}}
     {transformType=Translation, targetBone {boneName=None}}
    ]
    constraintWeights [1,1,1]
   }
   n71 = "Constraint - dyn_skirtSIde_02 (PPT)"(ComponentPose=@n70, Alpha=1, ConstraintWeights_0=0.7, ConstraintWeights_1=0.3, ConstraintWeights_2=0) {
    alphaBoolBlend {blendOption=Linear}
    boneToModify {boneName=dyn_skirtSIde_02}
    constraintSetup [{targetBone {boneName=dyn_skirtFrontSIde_02}}, {targetBone {boneName=calf_l}}, {transformType=Translation, targetBone {boneName=None}}]
    constraintWeights [1,1,1]
   }
   n72 = "Constraint - dyn_skirtSIde_03 (PPT)"(ComponentPose=@n71, Alpha=1, ConstraintWeights_0=0.3, ConstraintWeights_1=0.7, ConstraintWeights_2=0) {
    alphaBoolBlend {blendOption=Linear}
    boneToModify {boneName=dyn_skirtSIde_03}
    constraintSetup [{targetBone {boneName=dyn_skirtFrontSIde_03}}, {targetBone {boneName=calf_l}}, {transformType=Translation, targetBone {boneName=None}}]
    constraintWeights [1,1,1]
   }
   n73 = "Constraint - dyn_skirtSIde_04 (PPT)"(ComponentPose=@n72, Alpha=1, ConstraintWeights_0=0.3, ConstraintWeights_1=0.7, ConstraintWeights_2=0) {
    alphaBoolBlend {blendOption=Linear}
    boneToModify {boneName=dyn_skirtSIde_04}
    constraintSetup [{targetBone {boneName=dyn_skirtFrontSIde_04}}, {targetBone {boneName=calf_l}}, {transformType=Translation, targetBone {boneName=None}}]
    constraintWeights [1,1,1]
   }
   n74 = "Constraint - dyn_SkirtBackSide_Offset (PPT)"(ComponentPose=@n73, Alpha=1, ConstraintWeights_0=0.5, ConstraintWeights_1=0.5, ConstraintWeights_2=0) {
    alphaBoolBlend {blendOption=Linear}
    boneToModify {boneName=dyn_SkirtBackSide_Offset}
    constraintSetup [
     {targetBone {boneName=dyn_SkirtSide_Offset}}
     {targetBone {boneName=dyn_SkirtBack_Offset}}
     {transformType=Translation, targetBone {boneName=None}}
    ]
    constraintWeights [1,1,1]
   }
   n75 = "Constraint - dyn_skirtBackSIde_01 (PPT)"(ComponentPose=@n74, Alpha=1, ConstraintWeights_0=0.5, ConstraintWeights_1=0.5, ConstraintWeights_2=0) {
    alphaBoolBlend {blendOption=Linear}
    boneToModify {boneName=dyn_skirtBackSIde_01}
    constraintSetup [{targetBone {boneName=dyn_skirtSIde_01}}, {targetBone {boneName=dyn_SkirtBack_01}}, {transformType=Translation, targetBone {boneName=None}}]
    constraintWeights [1,1,1]
   }
   n76 = "Constraint - dyn_skirtBackSIde_02 (PPT)"(ComponentPose=@n75, Alpha=1, ConstraintWeights_0=0.5, ConstraintWeights_1=0.5, ConstraintWeights_2=0) {
    alphaBoolBlend {blendOption=Linear}
    boneToModify {boneName=dyn_skirtBackSIde_02}
    constraintSetup [{targetBone {boneName=dyn_skirtSIde_02}}, {targetBone {boneName=calf_l}}, {transformType=Translation, targetBone {boneName=None}}]
    constraintWeights [1,1,1]
   }
   n77 = "Constraint - dyn_skirtBackSIde_03 (PPT)"(ComponentPose=@n76, Alpha=1, ConstraintWeights_0=0.5, ConstraintWeights_1=0.5, ConstraintWeights_2=0) {
    alphaBoolBlend {blendOption=Linear}
    boneToModify {boneName=dyn_skirtBackSIde_03}
    constraintSetup [{targetBone {boneName=dyn_skirtSIde_03}}, {targetBone {boneName=calf_l}}, {transformType=Translation, targetBone {boneName=None}}]
    constraintWeights [1,1,1]
   }
   n78 = "Constraint - dyn_skirtBackSIde_04 (PPT)"(ComponentPose=@n77, Alpha=1, ConstraintWeights_0=0.5, ConstraintWeights_1=0.5, ConstraintWeights_2=0) {
    alphaBoolBlend {blendOption=Linear}
    boneToModify {boneName=dyn_skirtBackSIde_04}
    constraintSetup [{targetBone {boneName=dyn_skirtSIde_04}}, {targetBone {boneName=dyn_SkirtBack_04}}, {transformType=Translation, targetBone {boneName=None}}]
    constraintWeights [1,1,1]
   }
   n79 = "Constraint - dyn_SkirtBack_Offset (P)"(ComponentPose=@n78, Alpha=0.3, ConstraintWeights_0=1) {
    alphaBoolBlend {blendOption=Linear}
    boneToModify {boneName=dyn_SkirtBack_Offset}
    constraintSetup [{targetBone {boneName=dyn_SkirtBackSide_Offset}}]
    constraintWeights [1]
   }
   n80 = "Constraint - dyn_SkirtBack_01 (P)"(ComponentPose=@n79, Alpha=0.3, ConstraintWeights_0=1) {
    alphaBoolBlend {blendOption=Linear}
    boneToModify {boneName=dyn_SkirtBack_01}
    constraintSetup [{targetBone {boneName=dyn_SkirtBackSide_01}}]
    constraintWeights [1]
   }
   n81 = "Constraint - dyn_SkirtBack_02 (P)"(ComponentPose=@n80, Alpha=0.3, ConstraintWeights_0=1) {
    alphaBoolBlend {blendOption=Linear}
    boneToModify {boneName=dyn_SkirtBack_02}
    constraintSetup [{targetBone {boneName=dyn_SkirtBackSide_02}}]
    constraintWeights [1]
   }
   n82 = "Constraint - dyn_SkirtBack_03 (P)"(ComponentPose=@n81, Alpha=0.3, ConstraintWeights_0=1) {
    alphaBoolBlend {blendOption=Linear}
    boneToModify {boneName=dyn_SkirtBack_03}
    constraintSetup [{targetBone {boneName=dyn_SkirtBackSide_03}}]
    constraintWeights [1]
   }
   n83 = "Constraint - dyn_SkirtBack_04 (P)"(ComponentPose=@n82, Alpha=0.3, ConstraintWeights_0=1) {
    alphaBoolBlend {blendOption=Linear}
    boneToModify {boneName=dyn_SkirtBack_04}
    constraintSetup [{targetBone {boneName=dyn_SkirtBackSide_04}}]
    constraintWeights [1]
   }
   Save cached pose 'CrossConstraints'(Pose=Component To Local(ComponentPose=@n83)) {cacheName=CrossConstraints}
   SelectFloat(A=1, B=0, bPickA="float < float"!(A:double=1, B:double=Get PawnSpeed2D()) -> (ReturnValue:bool) {is=title K2Node_CallFunction})
   "To Float (Boolean)"(InBool=@n28)
   Get rigids alpha() -> (rigids alpha:double)
  ]}
 EventGraph {nodes [
   Event BlueprintUpdateAnimation()
   > n1 = Sequence!(execute) -> (then_0, then_1, then_2:Exec, then_3:Exec)
   > Set crouchAlpha(crouchAlpha:double="To Float (Boolean)"(InBool=Get bIsCrouching())) -> (Output_Get:double)
   n2 = Get PawnSpeed2D()
   > Set speedAlpha(speedAlpha:double="To Float (Boolean)"(InBool="float >= float"(A=@n2, B=405) {is=title K2Node_CallFunction})) -> (Output_Get:double)
   n3 = Branch(execute=@n1.then_1, Condition=Get bIsSkyDiving())
   > n4 = Branch(Condition=Get bIsParachuteOpened())
   Set gravity_override(execute=[@n3.else, @n4.then], gravity_override="0, 0,-980")
   > Set rigids alpha(rigids alpha:double=1) -> (Output_Get:double)
   n5 = Branch(execute=@n4.else, Condition="float > float"(A=Get PawnVelocityZ(), B=50) {is=title K2Node_CallFunction})
   > n6 = Set gravity_override(gravity_override="0, 0,200")
   Set rigids alpha(execute=[@n6.then, @n7.then], rigids alpha=1)
   n7 = Set gravity_override(execute=@n5.else, gravity_override="0, 0,-200")
   n8 = Get bRigidBodyEnabled()
   Set rban_disabled_dbno(
    execute=@n1.then_2
    rban_disabled_dbno:bool=AND Boolean!(A=NOT Boolean(A=@n8), B=Get bIsDBNO(), C:bool=Get bIsDBNOCarried()) -> (ReturnValue)
   ) -> (Output_Get:bool)
   n9 = Get bIsSkyDiving()
   n10 = Get bIsJumping()
   n11 = Get bIsFalling()
   > Set falling_notskydiving(falling_notskydiving:bool=AND Boolean!(A=OR Boolean!(A=@n10, B=@n11, C:bool) -> (ReturnValue), B=NOT Boolean(A=@n9)) -> (ReturnValue)) -> (Output_Get:bool)
   > Set parachute_opened_float(parachute_opened_float:double=SelectFloat(A=0.2, B=0, bPickA=Get bIsParachuteOpened())) -> (Output_Get:double)
   > Set skydive_swim(skydive_swim:bool=OR Boolean!(
     A=Get bIsSkyDiving()
     B=Get bIsSkydiveDiveMode()
     C=Get bIsSkydiveFloating()
     D:bool=Get bIsSurfaceSwimming()
     E:bool=Get bIsSwimSprinting() -> (bIsSwimSprinting:bool)
     F:bool=Get bIsParachuteOpened()
    ) -> (ReturnValue)) -> (Output_Get:bool)
  ]}
]
variables [
 compLinearAccScale {type=Vector}
 compLinearVelScale {type=Vector}
 crouchAlpha {type=number}
 falling_notskydiving {type=boolean}
 finalGravity {type=Vector}
 gravity_override {type=Vector}
 parachute_opened_float {type=number}
 random_stream {type=RandomStream, default {initialSeed=0, seed=0}}
 rban_disabled_dbno {type=boolean}
 rigids alpha {type=number}
 skydive_swim {type=boolean}
 skydiving vector force {type=Vector}
 speedAlpha {type=number}
 windVector {type=Vector}
]
