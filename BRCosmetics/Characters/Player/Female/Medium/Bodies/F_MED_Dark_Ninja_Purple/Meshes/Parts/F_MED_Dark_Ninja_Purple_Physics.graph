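# Physics asset for the hood of the purple Dark Ninja medium female player cosmetic, covering the head and its dynamic hood bones.
# Larger capsules form the top, side, lower-side and back hood segments, while small capsules sit on the head and base bones.
# Base bones join the head with free swing and twist, and their hood segments hang off them with narrower angular limits.
# The upper back segment constrains directly to the head without a base bone.
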PhysicsAsset /BRCosmetics/Characters/Player/Female/Medium/Bodies/F_MED_Dark_Ninja_Purple/Meshes/Parts/F_MED_Dark_Ninja_Purple_Physics
bodies {
 dyn_hood_back [{shapeType=Capsule, length=2.5, radius=3.25, center=<4.2814 0 0>, rotation=<rot 90 0 0>}]
 dyn_hood_back_btm [{shapeType=Capsule, length=2.5, radius=3.25, center=<4.3 0 0>, rotation=<rot 90 0 0>}]
 dyn_hood_base_back_btm [{shapeType=Capsule, length=1, radius=1}]
 dyn_hood_base_l [{shapeType=Capsule, length=1, radius=1}]
 dyn_hood_base_r [{shapeType=Capsule, length=1, radius=1}]
 dyn_hood_l [{shapeType=Capsule, length=2.5, radius=3.25, center=<4.3249 -0.0001 0>, rotation=<rot 90 0 0>}]
 dyn_hood_lo_base_l [{shapeType=Capsule, length=1, radius=1}]
 dyn_hood_lo_l [{shapeType=Capsule, length=2.5, radius=3.25, center=<4.4858 0 0.0001>, rotation=<rot 90 0 0>}]
 dyn_hood_lo_r [{shapeType=Capsule, length=2.5, radius=3.25, center=<4.4238 0 0>, rotation=<rot 90 0 0>}]
 dyn_hood_r [{shapeType=Capsule, length=2.5, radius=3.25, center=<4.3841 0 0>, rotation=<rot 90 0 0>}]
 dyn_hood_ro_base_r [{shapeType=Capsule, length=1, radius=1}]
 dyn_hood_top [{shapeType=Capsule, length=2.5, radius=3.25, center=<4.1325 0 0>, rotation=<rot 90 0 0>}]
 dyn_hood_top_base [{shapeType=Capsule, length=1, radius=1}]
 head [{shapeType=Capsule, length=1, radius=1}]
}
constraints [
 {bone1Name=dyn_hood_top, bone2Name=dyn_hood_top_base, swing1LimitDegrees=3, swing2LimitDegrees=3, twistLimitDegrees=3}
 {bone1Name=dyn_hood_l, bone2Name=dyn_hood_base_l, swing1LimitDegrees=5, swing2LimitDegrees=5, twistLimitDegrees=5}
 {bone1Name=dyn_hood_r, bone2Name=dyn_hood_base_r, swing1LimitDegrees=5, swing2LimitDegrees=5, twistLimitDegrees=5}
 {bone1Name=dyn_hood_lo_l, bone2Name=dyn_hood_lo_base_l, swing1LimitDegrees=5, swing2LimitDegrees=15, twistLimitDegrees=25}
 {bone1Name=dyn_hood_lo_r, bone2Name=dyn_hood_ro_base_r, swing1LimitDegrees=5, swing2LimitDegrees=15, twistLimitDegrees=25}
 {bone1Name=dyn_hood_back_btm, bone2Name=dyn_hood_base_back_btm, swing1LimitDegrees=5, swing2LimitDegrees=15, twistLimitDegrees=25}
 {
  bone1Name=dyn_hood_top_base
  bone2Name=head
  swing1LimitDegrees=45
  swing1Motion=Free
  swing2LimitDegrees=45
  swing2Motion=Free
  twistLimitDegrees=45
  twistMotion=Free
 }
 {
  bone1Name=dyn_hood_base_l
  bone2Name=head
  swing1LimitDegrees=45
  swing1Motion=Free
  swing2LimitDegrees=45
  swing2Motion=Free
  twistLimitDegrees=45
  twistMotion=Free
 }
 {
  bone1Name=dyn_hood_base_r
  bone2Name=head
  swing1LimitDegrees=45
  swing1Motion=Free
  swing2LimitDegrees=45
  swing2Motion=Free
  twistLimitDegrees=45
  twistMotion=Free
 }
 {bone1Name=dyn_hood_back, bone2Name=head, swing1LimitDegrees=10, swing2LimitDegrees=10, twistLimitDegrees=10}
 {
  bone1Name=dyn_hood_lo_base_l
  bone2Name=head
  swing1LimitDegrees=45
  swing1Motion=Free
  swing2LimitDegrees=45
  swing2Motion=Free
  twistLimitDegrees=45
  twistMotion=Free
 }
 {
  bone1Name=dyn_hood_ro_base_r
  bone2Name=head
  swing1LimitDegrees=45
  swing1Motion=Free
  swing2LimitDegrees=45
  swing2Motion=Free
  twistLimitDegrees=45
  twistMotion=Free
 }
 {
  bone1Name=dyn_hood_base_back_btm
  bone2Name=head
  swing1LimitDegrees=45
  swing1Motion=Free
  swing2LimitDegrees=45
  swing2Motion=Free
  twistLimitDegrees=45
  twistMotion=Free
 }
]
