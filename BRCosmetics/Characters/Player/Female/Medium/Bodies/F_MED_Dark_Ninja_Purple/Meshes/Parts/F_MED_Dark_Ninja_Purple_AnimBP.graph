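# Post-process style part AnimBP for the Dark Ninja Purple female body that drives its dynamic hood.
# It copies the attached parent's pose, layers both arms from Body_Mesh, nudges the right hood base bone, then simulates RigidBody physics.
# The nudge's alpha comes from right upper-arm pitch and yaw relative to the clavicle.
# The event graph also sets targeting and weapon-type aim values, which the AnimGraph never reads.
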AnimBlueprint /BRCosmetics/Characters/Player/Female/Medium/Bodies/F_MED_Dark_Ninja_Purple/Meshes/Parts/F_MED_Dark_Ninja_Purple_AnimBP
target_skeleton=&/F_MED_Dark_Ninja_Purple_Skeleton
graphs [
 AnimGraph {nodes [
   n1 = Copy Pose From Mesh(SourceMeshComponent=Get Body_Mesh())
   n2 = Copy Pose From Mesh() {bCopyCurves=true, bUseAttachedParent=true}
   n3 = "Transform (Modify) Bone - Bone: dyn_hood_ro_base_r"(
    ComponentPose=Local To Component(LocalPose=Layered blend per bone(BasePose=@n2, BlendPoses_0=@n1) {layerSetup [{branchFilters [{boneName=upperarm_l}, {boneName=upperarm_r}]}]})
    Alpha=Get hoodFix_alpha_r() -> (hoodFix_alpha_r:double)
    Translation="0,-1,3"
    Rotation="20,0,0"
    Scale="1,1,1"
   ) {
    rotationMode=BMM_Additive
    rotationSpace=BCS_BoneSpace
    translationMode=BMM_Additive
    translationSpace=BCS_BoneSpace
    alphaBoolBlend {blendOption=Linear}
    boneToModify {boneName=dyn_hood_ro_base_r}
    rotation {pitch=20}
    translation {y=-1, z=3}
   }
   n4 = RigidBody(ComponentPose=@n3) {
    bClampLinearTranslationLimitToRefPose=true
    bOverrideWorldGravity=true
    overridePhysicsAsset=&/F_MED_Dark_Ninja_Purple_Physics
    simulationSpace=BaseBoneSpace
    alphaBoolBlend {blendOption=Linear}
    baseBoneRef {boneName=root}
    componentLinearAccScale=<0.5 0.5 0.5>
    componentLinearVelScale=<0.2 0.2 0.2>
    overrideWorldGravity {z=-980}
   }
   Output Pose(Result=Component To Local(ComponentPose=@n4)) {layerGroup=None}
  ]}
 EventGraph {nodes [
   Event BlueprintUpdateAnimation()
   > n1 = Sequence!(execute) -> (then_0, then_1, then_2:Exec, then_3:Exec, then_4:Exec)
   > n2 = Branch(Condition=Get bIsTargeting())
   > Set hoodDynAlpha(hoodDynAlpha:double=0.35) -> (Output_Get:double)
   Set hoodDynAlpha(execute=@n2.else, hoodDynAlpha=1)
   n3 = Get armUp_R() -> (armUp_R:double)
   n4 = Get armForward_R() -> (armForward_R:double)
   Set hoodFix_alpha_r(
    execute=@n1.then_2
    hoodFix_alpha_r:double=float * float(A=MapRangeClamped(Value=@n3, InRangeA=-30, InRangeB=10), B=MapRangeClamped(Value=@n4, InRangeA=70, InRangeB=90))
   ) -> (Output_Get:double)
   n5 = GetOwnerSkeletalMeshComponent()
   n6 = BreakRotator(InRot=BreakTransform(InTransform=GetDeltaTransformFromRefPose(self=@n5, BoneName=upperarm_r, BaseName=clavicle_r)).Rotation)
   Set armForward_R(execute=@n1.then_1, armForward_R:double=@n6.Yaw) -> (Output_Get:double)
   > Set armUp_R(armUp_R:double=@n6.Pitch) -> (Output_Get:double)
   Set isAimingPistol(execute=@n1.then_3, isAimingPistol:double=0) -> (Output_Get:double)
   > Set isAimingRifle(isAimingRifle:bool=false) -> (Output_Get:bool)
   > n7 = Switch on EFortWeaponCoreAnimation!(execute, Selection:EFortWeaponCoreAnimation Enum=Get WeaponCoreAnim!() -> (WeaponCoreAnim)) -> (
    Melee:Exec
    Pistol:Exec
    Shotgun:Exec
    PaperBlueprint:Exec
    Rifle:Exec
    MeleeOneHand:Exec
    MachinePistol:Exec
    RocketLauncher:Exec
    GrenadeLauncher:Exec
    GoingCommando:Exec
    AssaultRifle:Exec
    TacticalShotgun:Exec
    SniperRifle:Exec
    TrapPlacement:Exec
    ShoulderLauncher:Exec
    AbilityDecoTool:Exec
    Crossbow:Exec
    C4:Exec
    RemoteControl:Exec
    DualWield:Exec
    AR_BullPup:Exec
    AR_ForwardGrip:Exec
    MedPackPaddles:Exec
    SMG_P90:Exec
    AR_DrumGun:Exec
    Consumable_Small:Exec
    Consumable_Large:Exec
    Balloon:Exec
    MountedTurret:Exec
    CreativeTool:Exec
    ExplosiveBow:Exec
    AshtonIndigo:Exec
    AshtonChicago:Exec
    MeleeDualWield:Exec
    Unarmed:Exec
   )
   Branch(execute=[@n7.AssaultRifle, @n7.Crossbow, @n7.GrenadeLauncher, @n7.Rifle, @n7.RocketLauncher, @n7.Shotgun, @n7.SniperRifle, @n7.TacticalShotgun])
   > Set isAimingRifle(isAimingRifle=true)
   Branch(execute=[@n7.MachinePistol, @n7.Pistol])
   > Set isAimingPistol(isAimingPistol=1)
   Event BlueprintInitializeAnimation()
   > Set Body_Mesh(Body_Mesh=GetOwnerSkeletalMeshComponent())
   > n8 = Cast To PlayerPawn_Athena(Object=TryGetPawnOwner())
   > n9 = Set Pawn Owner(Pawn Owner:Player Pawn Athena Object Reference=@n8) -> (Output_Get:Player Pawn Athena Object Reference)
   > Set Head_Mesh(Head_Mesh:Skeletal Mesh Component Object Reference=GetSkeletalMeshForPartType(self=@n9, PartType=Head)) -> (Output_Get:Skeletal Mesh Component Object Reference)
   n10 = Branch(Condition=Get isAimingRifle() -> (isAimingRifle:bool))
   n11 = BreakTransform(InTransform=GetDeltaTransformFromRefPose(BoneName=upperarm_l, BaseName=clavicle_l))
   > Set armUp_L(armUp_L:double=MapRangeClamped(Value=BreakRotator(InRot=@n11.Rotation).Yaw, InRangeA=22, InRangeB=16)) -> (Output_Get:double)
   n12 = GetPartSkeletalMeshComponent()
   Set rifleAimAlpha(
    execute=@n10.else
    rifleAimAlpha:double=SelectFloat(
     A=GetAlphaFromReferenceTransformDelta(
      SkeletalMeshComponent=@n12
      BoneName=upperarm_l
      BaseName=clavicle_l
      DeltaComponentType=QuaternionTwist
      InputRange=(X=22,Y=16)
      OutputRange=(X=50,Y=50)
      TwistAxis="1,0,0"
     )
     B=0
     bPickA=Get isAimingRifle()
    )
   ) -> (Output_Get:double)
   PrintString!(execute, InString="To String (Float)"(InDouble=0), bPrintToScreen, bPrintToLog, TextColor, Duration, Key) -> (then)
  ]}
]
variables [
 armForward_R {type=number}
 armUp_L {type=number}
 armUp_R {type=number}
 Body_Mesh {type=SkeletalMeshComponent}
 Head_Mesh {type=SkeletalMeshComponent}
 hoodDynAlpha {type=number}
 hoodFix_alpha_r {type=number}
 isAimingPistol {type=number}
 isAimingRifle {type=boolean}
 isWearingBackbling {type=boolean}
 LshoulderRotateBackAlpha {type=number}
 LshoulderRotateBackAlpha_Backbling {type=number}
 Pawn Owner {type=PlayerPawn_Athena_C}
 rifleAimAlpha {type=number}
]
